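; First-person player camera mode for the CrustyChain Feral Corgi pistol.
; On becoming active it calls the parent, validates the view target and camera manager, then waits one tick.
; It then sets the player controller's control rotation to the camera's pitch and yaw with zero roll.
; View rotation processing deliberately skips the parent and passes the rotation and delta through unchanged.
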(event EventOnBecomeActiveBP
  (params (ViewTarget :type Actor%20Object%20Reference))
  (bind _returnvalue (CallFunction|GetPlayerCameraManager))
  (bind (_returnvalue_roll _returnvalue_pitch _returnvalue_yaw) (Math|Rotator|BreakRotator (Camera|GetCameraRotation _returnvalue)))
  (bind _returnvalue_2 (Fortnite|GetFortPlayerControllerfromActor ViewTarget))
  (|CallParentFunction ViewTarget)
  (Utilities|IsValid ViewTarget
    (:Is%20Valid
      (Utilities|IsValid _returnvalue
        (:Is%20Valid
          (Utilities|FlowControl|DelayUntilNextTick)
          (Fortnite|KeepCGPathOptimizations _returnvalue_2)
          (Pawn|SetControlRotation _returnvalue_2 (Math|Rotator|MakeRotator 0.0 _returnvalue_pitch _returnvalue_yaw)))))))

(fn ProcessViewRotationBP
  (params (DeltaTime :type Float%20%28single-precision%29) (InViewRotation :type Rotator :default "0, 0, 0") (InDeltaRot :type Rotator :default "0, 0, 0"))
  ;[1 1 1] Deliberate No-Op; do NOT call parent
  (return InViewRotation InDeltaRot))
