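# MetaSound Source preset for the first-person sound of exiting aim-down-sights on the CrustyChain pistol.
# It inherits its graph from the shared MSS_Weapon_ADS_Out parent, supplying one CrustyChain wave, a volume and random pitch shift settings.
# Output is a mono one-shot using first-person foley equip attenuation and the weapon foley sound class and concurrency.
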MetaSoundSource /CrustyChain/Gameplay/Weapons/FeralCorgi_Pistol_CrustyChain/Audio/IP_MSS_Weapon_Pistol_CrustyChain_FP_ADS_Out
preset_parent=/FortMetasound/Weapons/MSS_Weapon_ADS_Out
inputs [
 MasterVolume {type=Float}
 Random.PitchShift.Base {type=Float}
 Random.PitchShift.Spread {type=Float}
 Sounds {type=WaveAsset:Array}
 UE.Source.OnPlay {type=Trigger}
]
interfaces [UE.OutputFormat.Mono:1.0, UE.Source.OneShot:1.0, UE.Source:1.1]
outputs [
 "UE.OutputFormat.Mono.Audio:0" {type=Audio}
 UE.Source.OneShot.OnFinished {type=Trigger}
]
pages [Default {input_defaults {
   MasterVolume=0.2
   Random.PitchShift.Base=0
   Random.PitchShift.Spread=0.5
   UE.Source.OnPlay=null
   Sounds [/CrustyChain/Assets/CrustyChain/IP_Jethro_CrustyChain_ADS_Out_01]
  }}]
properties {
 assetImportData=&self:AssetImportData
 attenuationSettings=/FortMetaSoundWeapon/Items/Shared/Settings/FPS/AS_Weapon_FP_Foley_Equip
 maxDistance=1400
 sampleRate=48000
 soundClassObject=&Fort_WeaponFoley
 concurrencySet [&CS_Fort_WeaponFoley]
}
