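# MetaSound Source preset that plays the mono, one-shot drop sound for the CrustyChain pistol weapon.
# It inherits its graph from the shared MSS_Item_Drop parent and supplies three drop wave variations, a master volume and random pitch-shift settings.
# It is routed through the Fort_Loot_Collect sound class and the CS_Fort_Item concurrency set.
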MetaSoundSource /CrustyChain/Gameplay/Weapons/FeralCorgi_Pistol_CrustyChain/Audio/IP_MSS_Weapon_Pistol_CrustyChain_FP_Drop
preset_parent=/FortMetasound/Items/MSS_Item_Drop
inputs [
 MasterVolume {type=Float}
 Random.PitchShift.Base {type=Float}
 Random.PitchShift.Spread {type=Float}
 Sounds {type=WaveAsset:Array}
 UE.Source.OnPlay {type=Trigger}
]
interfaces [UE.OutputFormat.Mono:1.0, UE.Source.OneShot:1.0, UE.Source:1.1]
outputs [
 "UE.OutputFormat.Mono.Audio:0" {type=Audio}
 UE.Source.OneShot.OnFinished {type=Trigger}
]
pages [Default {input_defaults {
   MasterVolume=0.1
   Random.PitchShift.Base=0
   Random.PitchShift.Spread=0.5
   UE.Source.OnPlay=null
   Sounds [
    /CrustyChain/Assets/CrustyChain/IP_Jethro_CrustyChain_Drop_01
    /CrustyChain/Assets/CrustyChain/IP_Jethro_CrustyChain_Drop_02
    /CrustyChain/Assets/CrustyChain/IP_Jethro_CrustyChain_Drop_03
   ]
  }}]
properties {
 assetImportData=&self:AssetImportData
 sampleRate=48000
 soundClassObject=&Fort_Loot_Collect
 concurrencySet [&CS_Fort_Item]
}
