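# MetaSound Source preset for the first-person fire sound of the CrustyChain pistol, output in 5.1.
# It inherits its graph from MSS_FPS_Weapon_Fire_1p and overrides inputs for layer sounds, pitch and volume randomization, aim-down-sights mix and worldization.
# Only main fire and mechanical layers have waves assigned; low-ammo, shell, suppressor, sweetener and stop-fire arrays are empty.
# It uses the shared local first-person gunfire concurrency set and ranged-weapon fire sound class.
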MetaSoundSource /CrustyChain/Gameplay/Weapons/FeralCorgi_Pistol_CrustyChain/Audio/IP_MSS_Weapon_Pistol_CrustyChain_FP_Fire_1p
preset_parent=/FortMetaSoundWeapon/Items/Metasounds/FPS/MSS_FPS_Weapon_Fire_1p
inputs [
 bIsFiring {type=Bool}
 BurstRate {type=Time}
 bUseSingleVoice {type=Bool}
 FireBurst {type=Trigger}
 FireSingleShot {type=Trigger}
 FPSMix.7 Weapon Output Volume {type=Float}
 FPSMIX.ADS.Filter.Bandwidth {type=Float}
 FPSMIX.ADS.Filter.CutoffFreq {type=Float}
 FPSMIX.ADS.Filter.GainADS {type=Float}
 FPSMIX.ADS.Filter.GainHipFire {type=Float}
 FPSMIX.ADS.Filter.Wave Shaper Main Fire Sub ADS Fire Amount {type=Float}
 FPSMIX.ADS.Filter.Wave Shaper Mech ADS Fire Amount {type=Float}
 FPSMix.ADS.LFEVolume {type=Float}
 FPSMix.CenterVolume {type=Float}
 FPSMix.LeftSurroundVolume {type=Float}
 FPSMix.LeftVolume {type=Float}
 FPSMix.LFEVolume {type=Float}
 FPSMix.RightSurroundVolume {type=Float}
 FPSMix.RightVolume {type=Float}
 FPSMix.Worldization.ActivationThreshold {type=Float}
 FPSMix.Worldization.EnvelopeMaxDecayTime {type=Float}
 FPSMix.Worldization.EnvelopeMinDecayTime {type=Float}
 LowAmmo.ADS.VolumeMultiply {type=Float}
 LowAmmo.Random.Pitch.Max {type=Float}
 LowAmmo.Random.Pitch.Min {type=Float}
 LowAmmo.Random.Volume.Max {type=Float}
 LowAmmo.Random.Volume.Min {type=Float}
 LowAmmo.Sounds {type=WaveAsset:Array}
 LowAmmo.Threshold {type=Int32}
 LowAmmo.Volume {type=Float}
 MainFire.ADS.VolumeMultiply {type=Float}
 MainFire.Random.Pitch.Max {type=Float}
 MainFire.Random.Pitch.Min {type=Float}
 MainFire.Random.Volume.Max {type=Float}
 MainFire.Random.Volume.Min {type=Float}
 MainFire.Sounds {type=WaveAsset:Array}
 MainFire.Volume {type=Float}
 Mech.ADSVolumeMultiply {type=Float}
 Mech.Random.Pitch.Max {type=Float}
 Mech.Random.Pitch.Min {type=Float}
 Mech.Random.Volume.Max {type=Float}
 Mech.Random.Volume.Min {type=Float}
 Mech.Sounds {type=WaveAsset:Array}
 Mech.Volume {type=Float}
 OnStop {type=Trigger}
 OnStopFire.Sounds {type=WaveAsset:Array}
 OnStopFire.Volume {type=Float}
 RetriggerDelayTime {type=Time}
 RoundsPerBurst {type=Int32}
 Shells.ADS.VolumeMultiply {type=Float}
 Shells.DelayTime {type=Time}
 Shells.PanAmount {type=Float}
 Shells.Random.Pitch.Max {type=Float}
 Shells.Random.Pitch.Min {type=Float}
 Shells.Random.Volume.Max {type=Float}
 Shells.Random.Volume.Min {type=Float}
 Shells.Sounds {type=WaveAsset:Array}
 Shells.Volume {type=Float}
 StopBurst {type=Trigger}
 Suppressor.BaseWeaponMix {type=Float}
 Suppressor.EnvelopeDecayTime {type=Time}
 Suppressor.MasterVolume {type=Float}
 Suppressor.Random.Pitch.Max {type=Float}
 Suppressor.Random.Pitch.Min {type=Float}
 Suppressor.Random.Volume.Max {type=Float}
 Suppressor.Random.Volume.Min {type=Float}
 Suppressor.Sounds {type=WaveAsset:Array}
 Suppressor.SuppressorVolume {type=Float}
 Sweetener.ADS.VolumeMultiply {type=Float}
 Sweetener.Random.Pitch.Max {type=Float}
 Sweetener.Random.Pitch.Min {type=Float}
 Sweetener.Random.Volume.Max {type=Float}
 Sweetener.Random.Volume.Min {type=Float}
 Sweetener.Sounds {type=WaveAsset:Array}
 Sweetener.Volume {type=Float}
 UE.Source.OnPlay {type=Trigger}
 Weapon.AmmoRemaining {type=Int32}
 Weapon.ClipSize {type=Int32}
 Weapon.FireRate {type=Float}
 Weapon.FireStartTime {type=Float}
 Weapon.IsAimingDownSights {type=Bool}
 Weapon.IsSuppressorEnabled {type=Bool}
]
interfaces [Fort.Weapon:1.0, UE.OutputFormat.5dot1:1.0, UE.Source.OneShot:1.0, UE.Source:1.1]
outputs [
 bOnAlmostFinished {type=Bool}
 OnGunshotOut {type=Trigger}
 "UE.OutputFormat.5dot1.Audio:0" {type=Audio}
 "UE.OutputFormat.5dot1.Audio:1" {type=Audio}
 "UE.OutputFormat.5dot1.Audio:2" {type=Audio}
 "UE.OutputFormat.5dot1.Audio:3" {type=Audio}
 "UE.OutputFormat.5dot1.Audio:4" {type=Audio}
 "UE.OutputFormat.5dot1.Audio:5" {type=Audio}
 UE.Source.OneShot.OnFinished {type=Trigger}
]
pages [Default {input_defaults {
   BurstRate=0.4
   FPSMIX.ADS.Filter.Bandwidth=1
   FPSMIX.ADS.Filter.CutoffFreq=6000
   FPSMIX.ADS.Filter.GainADS=1.5
   FPSMIX.ADS.Filter.GainHipFire=1.2
   FPSMIX.ADS.Filter.Wave Shaper Main Fire Sub ADS Fire Amount=1.1
   FPSMIX.ADS.Filter.Wave Shaper Mech ADS Fire Amount=1.1
   FPSMix.7 Weapon Output Volume=0.75
   FPSMix.ADS.LFEVolume=1.5
   FPSMix.CenterVolume=1
   FPSMix.LFEVolume=1
   FPSMix.LeftSurroundVolume=1
   FPSMix.LeftVolume=2
   FPSMix.RightSurroundVolume=1
   FPSMix.RightVolume=2
   FPSMix.Worldization.ActivationThreshold=0.8
   FPSMix.Worldization.EnvelopeMaxDecayTime=2
   FPSMix.Worldization.EnvelopeMinDecayTime=1
   FireBurst=null
   FireSingleShot=null
   LowAmmo.ADS.VolumeMultiply=1
   LowAmmo.Random.Pitch.Max=0.7
   LowAmmo.Random.Pitch.Min=0.7
   LowAmmo.Random.Volume.Max=1.05
   LowAmmo.Random.Volume.Min=0.95
   LowAmmo.Threshold=4
   LowAmmo.Volume=0.6
   MainFire.ADS.VolumeMultiply=1
   MainFire.Random.Pitch.Max=1
   MainFire.Random.Pitch.Min=1
   MainFire.Random.Volume.Max=1.05
   MainFire.Random.Volume.Min=0.95
   MainFire.Volume=1
   Mech.ADSVolumeMultiply=1.1
   Mech.Random.Pitch.Max=2
   Mech.Random.Pitch.Min=1.9
   Mech.Random.Volume.Max=1.05
   Mech.Random.Volume.Min=0.95
   Mech.Volume=1
   OnStop=null
   OnStopFire.Volume=0
   RetriggerDelayTime=2
   RoundsPerBurst=2
   Shells.ADS.VolumeMultiply=2
   Shells.DelayTime=0
   Shells.PanAmount=0.5
   Shells.Random.Pitch.Max=1
   Shells.Random.Pitch.Min=1
   Shells.Random.Volume.Max=1.05
   Shells.Random.Volume.Min=0.95
   Shells.Volume=0.25
   StopBurst=null
   Suppressor.BaseWeaponMix=0.6
   Suppressor.EnvelopeDecayTime=0.195
   Suppressor.MasterVolume=1
   Suppressor.Random.Pitch.Max=1.05
   Suppressor.Random.Pitch.Min=0.95
   Suppressor.Random.Volume.Max=1.05
   Suppressor.Random.Volume.Min=0.95
   Suppressor.SuppressorVolume=0
   Sweetener.ADS.VolumeMultiply=1.5
   Sweetener.Random.Pitch.Max=2
   Sweetener.Random.Pitch.Min=2
   Sweetener.Random.Volume.Max=1.05
   Sweetener.Random.Volume.Min=0.95
   Sweetener.Volume=0.8
   UE.Source.OnPlay=null
   Weapon.AmmoRemaining=30
   Weapon.ClipSize=30
   Weapon.FireRate=0.1515
   Weapon.FireStartTime=0
   Weapon.IsAimingDownSights=null
   Weapon.IsSuppressorEnabled=null
   bIsFiring=null
   bUseSingleVoice=null
   LowAmmo.Sounds []
   MainFire.Sounds [
    /CrustyChain/Assets/CrustyChain/IP_Jethro_CrustyChain_Fire_1P_Close_01
    /CrustyChain/Assets/CrustyChain/IP_Jethro_CrustyChain_Fire_1P_Close_02
    /CrustyChain/Assets/CrustyChain/IP_Jethro_CrustyChain_Fire_1P_Close_03
    /CrustyChain/Assets/CrustyChain/IP_Jethro_CrustyChain_Fire_1P_Close_04
    /CrustyChain/Assets/CrustyChain/IP_Jethro_CrustyChain_Fire_1P_Close_05
   ]
   Mech.Sounds [
    /CrustyChain/Assets/CrustyChain/IP_Jethro_CrustyChain_Fire_Close_Mech_01
    /CrustyChain/Assets/CrustyChain/IP_Jethro_CrustyChain_Fire_Close_Mech_02
    /CrustyChain/Assets/CrustyChain/IP_Jethro_CrustyChain_Fire_Close_Mech_03
    /CrustyChain/Assets/CrustyChain/IP_Jethro_CrustyChain_Fire_Close_Mech_04
   ]
   OnStopFire.Sounds []
   Shells.Sounds []
   Suppressor.Sounds []
   Sweetener.Sounds []
  }}]
properties {
 assetImportData=&self:AssetImportData
 numChannels=6
 outputFormat=FiveDotOne
 qualitySetting=Fort_Weapons
 sampleRate=48000
 soundClassObject=&Fort_WeaponRanged_Fire_1P_PSM
 volume=0.8
 concurrencyOverrides {maxCount=2, resolutionRule=StopOldest}
 concurrencySet [/FortMetaSoundWeapon/Items/Shared/Settings/FPS/CS_Weapon_FP_Gunfire_Local]
}
